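# Physics asset for the PER_2 hair part of the medium female Elastic player body, defining collision and joints for its hair bones.
# Capsules cover the head, hair root, bun and short left and right "fr" strands of one to three segments.
# The hair root joins the head with free swing and twist, and the bun and strands branch from that root.
# Strand joints have narrow angular limits, and the bun's are tighter still.
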PhysicsAsset /Game/Characters/Player/Female/Medium/Bodies/F_MED_Elastic/Meshes/Parts/PER_Parts/F_MED_Elastic_PER_2_Physics
bodies {
 head [{shapeType=Capsule, length=5.5, radius=5.75, center=<1.7572 3.7604 0>, rotation=<rot 90 0 0>}]
 per2_hair [{shapeType=Capsule, length=1, radius=1}]
 per2_hair_bun [{shapeType=Capsule, length=0.5, radius=2.25, center=<-1.1305 0 0>, rotation=<rot 90 0 0>}]
 per2_hair_fr_A_a_l [{shapeType=Capsule, length=2.75, radius=0.5, center=<1.8845 0 0>, rotation=<rot 90 0 0>}]
 per2_hair_fr_A_a_r [{shapeType=Capsule, length=1, radius=1, center=<1.3711 0 0>, rotation=<rot 90 0 0>}]
 per2_hair_fr_A_b_l [{shapeType=Capsule, length=2, radius=0.75, center=<1.6512 0 0>, rotation=<rot 90 0 0>}]
 per2_hair_fr_A_c_l [{shapeType=Capsule, length=1.5, radius=1.25, center=<1.9262 0 0>, rotation=<rot 90 0 0>}]
 per2_hair_fr_B_a_l [{shapeType=Capsule, length=2, radius=1, center=<1.8481 0 0>, rotation=<rot 90 0 0>}]
 per2_hair_fr_B_a_r [{shapeType=Capsule, length=1, radius=1, center=<1.3711 0 0>, rotation=<rot 90 0 0>}]
 per2_hair_fr_B_b_l [{shapeType=Capsule, length=2.5, radius=1, center=<2.0324 0 0>, rotation=<rot 90 0 0>}]
 per2_hair_fr_C_a_r [{shapeType=Capsule, length=1, radius=1, center=<1.0882 0 0>, rotation=<rot 90 0 0>}]
 per2_hair_fr_C_b_r [{shapeType=Capsule, length=1.9462, radius=1.0538, center=<1.9211 0 0>, rotation=<rot 90 0 0>}]
 per2_hair_fr_D_a_r [{shapeType=Capsule, length=1.5, radius=0.75, center=<1.5882 0 0>, rotation=<rot 90 0 0>}]
 per2_hair_fr_D_b_r [{shapeType=Capsule, length=2, radius=0.75, center=<1.2777 0 0>, rotation=<rot 90 0 0>}]
}
constraints [
 {
  bone1Name=per2_hair
  bone2Name=head
  swing1LimitDegrees=45
  swing1Motion=Free
  swing2LimitDegrees=45
  swing2Motion=Free
  twistLimitDegrees=45
  twistMotion=Free
 }
 {bone1Name=per2_hair_fr_A_a_l, bone2Name=per2_hair, swing1LimitDegrees=15, swing2LimitDegrees=15, twistLimitDegrees=5}
 {bone1Name=per2_hair_fr_A_b_l, bone2Name=per2_hair_fr_A_a_l, swing1LimitDegrees=15, swing2LimitDegrees=15, twistLimitDegrees=5}
 {bone1Name=per2_hair_fr_A_c_l, bone2Name=per2_hair_fr_A_b_l, swing1LimitDegrees=15, swing2LimitDegrees=15, twistLimitDegrees=5}
 {bone1Name=per2_hair_fr_B_a_l, bone2Name=per2_hair, swing1LimitDegrees=15, swing2LimitDegrees=15, twistLimitDegrees=5}
 {bone1Name=per2_hair_fr_B_b_l, bone2Name=per2_hair_fr_B_a_l, swing1LimitDegrees=15, swing2LimitDegrees=15, twistLimitDegrees=5}
 {bone1Name=per2_hair_fr_A_a_r, bone2Name=per2_hair, swing1LimitDegrees=15, swing2LimitDegrees=15, twistLimitDegrees=15}
 {bone1Name=per2_hair_fr_B_a_r, bone2Name=per2_hair, swing1LimitDegrees=15, swing2LimitDegrees=15, twistLimitDegrees=15}
 {bone1Name=per2_hair_fr_C_a_r, bone2Name=per2_hair, swing1LimitDegrees=15, swing2LimitDegrees=15, twistLimitDegrees=5}
 {bone1Name=per2_hair_fr_C_b_r, bone2Name=per2_hair_fr_C_a_r, swing1LimitDegrees=15, swing2LimitDegrees=15, twistLimitDegrees=5}
 {bone1Name=per2_hair_fr_D_a_r, bone2Name=per2_hair, swing1LimitDegrees=15, swing2LimitDegrees=15, twistLimitDegrees=5}
 {bone1Name=per2_hair_fr_D_b_r, bone2Name=per2_hair_fr_D_a_r, swing1LimitDegrees=15, swing2LimitDegrees=15, twistLimitDegrees=5}
 {bone1Name=per2_hair_bun, bone2Name=per2_hair, swing1LimitDegrees=5, swing2LimitDegrees=5, twistLimitDegrees=1}
]
